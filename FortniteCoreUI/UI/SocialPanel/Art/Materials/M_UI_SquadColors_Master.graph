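# Translucent UI master material that tints squad logos and ribbons for the social panel with three squad colors, selected by the IsLogo static switch.
# Logo mode layers the colors and a blended midtone through the logo texture's RGB channels, using alpha as opacity.
# Ribbon mode smoothsteps a rotated stripe texture between midtone and secondary color, with optional dimming and full opacity.
# Both modes then support desaturation and an inactive color fill.
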Material /FortniteCoreUI/UI/SocialPanel/Art/Materials/M_UI_SquadColors_Master
graph {
 properties {
  bUsedWithStaticMesh=false
  blendMode=BLEND_Translucent
  materialDomain=MD_UI
  pixelDepthOffsetMode=PDOM_Legacy
  usedShadingModels=Default Lit
 }
 nodes [
  n1 = VectorParameter() {parameterName=SquadRGB_3, defaultValue=<rgba 0 0.0075 0.0331 1>}
  n2 = VectorParameter() {parameterName=SquadRGB_2, defaultValue=<rgba 1 0.1195 0.4072 1>}
  n3 = LinearInterpolate(A=@n1.RGB, B=@n2.RGB, Alpha=ScalarParameter() {defaultValue=0.25, parameterName=MidtoneBlend})
  n4 = TextureSampleParameter2D() {parameterName=LogoTexture, samplerType=SAMPLERTYPE_LinearColor, texture=/FortniteCoreUI/UI/SocialPanel/Art/Textures/T_UI_Logo_Bear}
  n5 = LinearInterpolate(
   A=LinearInterpolate(A=@n1.RGB, B=LinearInterpolate(A=@n3, B=@n2.RGB, Alpha=@n4.G), Alpha=@n4.B)
   B=VectorParameter() {parameterName=SquadRGB_1, defaultValue=<rgba 1 1 1 1>}.RGB
   Alpha=@n4.R
  )
  n6 = TextureSampleParameter2D(UVs=Rotator(Coordinate=TextureCoordinate(), Time=ScalarParameter() {defaultValue=0.25, parameterName=RibbonStripeRotation})) {parameterName=RibbonTexture, samplerType=SAMPLERTYPE_LinearColor, texture=/FortniteCoreUI/UI/SocialPanel/Art/Textures/T_UI_Ribbon_Bear}
  n7 = ScalarParameter() {defaultValue=0.075, parameterName=RibbonSharpness}
  n8 = ScalarParameter() {defaultValue=0.5, parameterName=RibbonThickness, sliderMax=1}
  n9 = Add(A=Multiply(A=@n7) {constB=-1}, B=@n8)
  n10 = Add(A=@n7, B=@n8)
  n11 = LinearInterpolate(
   A=LinearInterpolate(
    A=@n3
    B=@n2.RGB
    Alpha=MaterialFunctionCall(Alpha=@n6.R, Min=@n9, Max=@n10) {materialFunction=&SmoothStep}.Result
   )
   B=VectorParameter() {parameterName=RibbonDimRGB, defaultValue=<rgba 0 0.0168 0.0887 1>}.RGB
   Alpha=ScalarParameter() {parameterName=DimRibbon}
  )
  LinearInterpolate(
   A=Desaturation(
    None=StaticSwitchParameter(True=@n5, False=@n11) {defaultValue=true, parameterName=IsLogo}
    Fraction=ScalarParameter() {parameterName=DesatFactor}
   )
   B=VectorParameter() {parameterName=InactiveRGB, defaultValue=<rgba 0 0.004 0.0331 1>}.RGB
   Alpha=ScalarParameter() {parameterName=InactiveFill}
  )
  StaticSwitchParameter(True=@n4.A, False=Constant() {r=1}) {defaultValue=true, parameterName=IsLogo}
 ]
}
